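# Conversation database for the Peely (Banana) NPC, defining greetings and the dialogue branches offered when a player talks to him.
# The entry tag leads to one Speech node with a default message plus contextual lines gated by first-ever conversation, per-match conversation count and equipped outfit.
# First-meeting and outfit lines take priority from curve rows.
# Five link nodes jump by tag to external NPC quest, basic service, random service, Season 37 rescue and Season 39 quest conversations.
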ConversationDatabase /NPCLibrary/NPCs/Banana/NPCLibrary_Conversation_Banana
entry_points [{tag=Conversation.Entry.Banana, destinations [#1]}]
nodes [
 ConversationEntryPointNode_7 {
  id=#1
  class=ConversationEntryPointNode
  outputs [#2]
  properties {entryTag {tagName=Conversation.Entry.Banana}}
 }
 Speech_C_0 {
  id=#2
  class=Speech_C
  outputs [
   #3
   #4
   #5
   #6
   #7
  ]
  properties {
   nodeName=""
   generalConfig {
    defaultMessage=(Peely waves excitedly to see you.)
    contextualMessages [
     {
      message=(Peely waves excitedly to see you.)
      contextRequirements [{requirement {class=FortControllerRequirement_IsFirstEverConversationWithNPC}}]
      priority {curve {rowName=Tandem.Speech.Priority.FirstEverInteraction}}
     }
     {
      message=(Peely smiles shyly and shows you his new acting headshot.)
      contextRequirements [{requirement {class=FortControllerRequirement_CheckNPCMatchConversationCount, properties {nPCUniqueIdTag {tagName=AISpawnerData.Type.Tandem.Banana}}}}]
      priority {curve {curveTable=null, rowName=None}}
     }
     {
      message=(Peely gestures to a nearby Peelians poster. He's a movie star!)
      contextRequirements [{requirement {class=FortControllerRequirement_CheckNPCMatchConversationCount, properties {nPCUniqueIdTag {tagName=AISpawnerData.Type.Tandem.Banana}}}}]
      priority {curve {curveTable=null, rowName=None}}
     }
     {
      message=(Peely gapes in awe at the wonder around him.)
      contextRequirements [{requirement {class=FortControllerRequirement_CheckNPCMatchConversationCount, properties {nPCUniqueIdTag {tagName=AISpawnerData.Type.Tandem.Banana}}}}]
      priority {curve {curveTable=null, rowName=None}}
     }
     {
      message=(Peely grins and takes a selfie with you. Red carpet buddies!)
      contextRequirements [{requirement {class=FortControllerRequirement_HasCID, properties {allowedCIds [/BRCosmetics/Athena/Items/Cosmetics/Characters/Character_RowLiaison_Cafe]}}}]
      priority {curve {rowName=Tandem.Speech.Priority.OutfitReaction}}
     }
     {
      message=(Peely scratches his head. He's lost his script again...)
      contextRequirements [{requirement {
         class=FortControllerRequirement_HasCID
         properties {allowedCIds [/BRCosmetics/Athena/Items/Cosmetics/Characters/Character_DerangedMile_Side]}
        }}]
      priority {curve {rowName=Tandem.Speech.Priority.OutfitReaction}}
     }
     {
      message=(Peely attempts a sick stunt. It goes badly.)
      contextRequirements [{requirement {class=FortControllerRequirement_HasCID, properties {allowedCIds [/BRCosmetics/Athena/Items/Cosmetics/Characters/Character_ZenCrash]}}}]
      priority {curve {rowName=Tandem.Speech.Priority.OutfitReaction}}
     }
     {
      message=(Peely salutes. Yes, director, sir!)
      contextRequirements [{requirement {class=FortControllerRequirement_HasCID, properties {allowedCIds [/BRCosmetics/Athena/Items/Cosmetics/Characters/Character_FloralMane_Sand]}}}]
      priority {curve {rowName=Tandem.Speech.Priority.OutfitReaction}}
     }
     {
      message=(Peely pulls you into a tight hug. Best friends!)
      contextRequirements [{requirement {
         class=FortControllerRequirement_HasCID
         properties {allowedCIds [&Character_UndergroundRebel_Fashion]}
        }}]
      priority {curve {rowName=Tandem.Speech.Priority.OutfitReaction}}
     }
     {
      message=(Peely looks at you with pride in his eyes.)
      contextRequirements [{requirement {
         class=FortControllerRequirement_HasCID
         properties {allowedCIds [
           &Character_IronLilac_Sly
           &Character_BrightDisk
           &Character_VoidRedemption_Rebel
          ]}
        }}]
      priority {curve {rowName=Tandem.Speech.Priority.OutfitReaction}}
     }
    ]
   }
  }
 }
 ConversationLinkNode_3 {id=#7, class=ConversationLinkNode, properties {remoteEntryTag {tagName=Conversation.Entry.RandomServices}}}
 ConversationLinkNode_4 {
  id=#6
  class=ConversationLinkNode
  properties {remoteEntryTag {tagName=Conversation.Entry.Season39.Banana.BRQuests.01}}
 }
 ConversationLinkNode_2 {
  id=#5
  class=ConversationLinkNode
  properties {remoteEntryTag {tagName=Conversation.Entry.Season37.NPCRescues.Combat}}
 }
 ConversationLinkNode_0 {id=#4, class=ConversationLinkNode, properties {remoteEntryTag {tagName=Conversation.Entry.BasicServices}}}
 ConversationLinkNode_1 {id=#3, class=ConversationLinkNode, properties {remoteEntryTag {tagName=Conversation.Entry.NPCLibrary.Quests}}}
]
